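# Translucent, per-pixel lit glass material for the female Space Wanderer pickaxe.
# Diffuse, normal and spec textures (spec green drives metallic, specular and roughness) feed PanningEmissive, then MF_TechArtOverrides_accessories.
# Opacity adds a view-angle inner/outer blend to a local-bounds Y gradient, then lerps by the Mask texture that also drives PanningEmissive.
# The gradient's divisor subtracts bounds Max from itself, so it is zero.
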Material /BRCosmetics/Weapons/FORT_Melee/Pickaxe_Space_Wanderer_Female/Materials/M_SPaceWanderer_PckaxeGlass
graph {
 properties {
  bUseMaterialAttributes=true
  bUsedWithMeshDeformer=true
  bUsedWithSkeletalMesh=true
  blendMode=BLEND_Translucent
  pixelDepthOffsetMode=PDOM_Legacy
  translucencyLightingMode=TLM_SurfacePerPixelLighting
  translucentShadowDensityScale=0
  usedShadingModels=Default Lit
 }
 nodes [
  LinearInterpolate()
  n1 = Clamp(None=DotProduct(A=PixelNormalWS(), B=CameraVectorWS()))
  n2 = ScalarParameter() {defaultValue=2, parameterName=ViewDepedantFalloff}
  n3 = PreSkinnedLocalBounds()
  n4 = Subtract(A=VertexInterpolator(VS=PreSkinnedPosition()).PS, B=@n3.Min)
  n5 = Subtract(A=@n3.Max, B=@n3.Max)
  n6 = Power(Base=ComponentMask(None=Divide(A=@n4, B=@n5)) {g=true})
  n7 = Add(
   A=LinearInterpolate(A=ScalarParameter() {parameterName=inner Opacity}, B=ScalarParameter() {parameterName=outer Opacity}, Alpha=Power(Base=@n1, Exp=@n2))
   B=LinearInterpolate(B=ScalarParameter() {defaultValue=0.3, parameterName=BackOpacity Add}, Alpha=OneMinus(None=@n6)) {constB=0.15}
  )
  n8 = TextureSampleParameter2D() {
   parameterName=Mask
   samplerType=SAMPLERTYPE_LinearGrayscale
   texture=/BRCosmetics/Weapons/FORT_Melee/Pickaxe_Space_Wanderer_Female/Textures/spacewanderer_mask2h
  }
  n9 = TextureSampleParameter2D() {
   parameterName=Spec
   samplerType=SAMPLERTYPE_Masks
   texture=/BRCosmetics/Weapons/FORT_Melee/Pickaxe_Space_Wanderer_Female/Textures/T_Space_Wanderer_Blue_Female_S
  }
  n10 = TextureSampleParameter2D() {
   parameterName=Normal
   samplerType=SAMPLERTYPE_Normal
   texture=/BRCosmetics/Weapons/FORT_Melee/Pickaxe_Space_Wanderer_Female/Textures/T_Space_Wanderer_Female_N
  }
  n11 = SetMaterialAttributes!(
   MaterialAttributes
   Base Color=TextureSampleParameter2D() {parameterName=Diffuse, texture=/BRCosmetics/Weapons/FORT_Melee/Pickaxe_Space_Wanderer_Female/Textures/T_Space_Wanderer_Female_D}.RGB
   Metallic=@n9.G
   Specular=@n9.G
   Roughness=@n9.G
   Anisotropy
   Emissive Color
   Opacity Mask
   Normal=@n10.RGB
   Tangent
  ) {attributeSetTypes [
    #1
    #2
    #3
    #4
    #5
    #6
    #7
    #8
    #9
   ]}
  MaterialFunctionCall(In=SetMaterialAttributes!(
    MaterialAttributes=MaterialFunctionCall!(Alpha=@n8.R, isVehicle, Gmap, UVOffset, Local Normal, Local Position, EmissiveColorOverride, Material=@n11) {materialFunction=/Game/Weapons/WeaponSkins/MaterialFunction/PanningEmissive}.Result
    Opacity=LinearInterpolate(A=@n7, Alpha=@n8.R)
   ) {attributeSetTypes [#10]}) {materialFunction=&MF_TechArtOverrides_accessories}
 ]
}
