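# Physics asset for the female Space Wanderer pickaxe, providing collision bodies and a joint for its dynamic cable bones.
# The dyn_cable body is a small offset capsule, joined to dyn_master_cable by a single constraint with matching swing and twist limit angles.
# The dyn_master_cable body combines two small capsules at the bone with one very large offset capsule, far bigger than the cable shapes.
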PhysicsAsset /BRCosmetics/Weapons/FORT_Melee/Pickaxe_Space_Wanderer_Female/Meshes/Pickaxe_Space_Wanderer_Female_Physics
bodies {
 dyn_cable [{shapeType=Capsule, length=4, radius=0.5, center=<2.6417 0.0413 -1.0481>, rotation=<rot -5.6719 -89.999 5.1189>}]
 dyn_master_cable [
  {shapeType=Capsule, length=1, radius=1}
  {shapeType=Capsule, length=0.5, radius=0.5}
  {shapeType=Capsule, length=90.0889, radius=38.8835, center=<0 16.0382 -59.6304>}
 ]
}
constraints [{bone1Name=dyn_cable, bone2Name=dyn_master_cable, swing1LimitDegrees=45, swing2LimitDegrees=45, twistLimitDegrees=45}]
